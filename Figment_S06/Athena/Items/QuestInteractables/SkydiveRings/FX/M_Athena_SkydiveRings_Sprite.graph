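# Translucent unlit, two-sided particle sprite for skydive ring quest interactables, tinted by particle color.
# Two sphere masks with dynamic-parameter radii carve a soft radial shape.
# Param3 blends in panning cloud and lightning textures on radially remapped, rotated UVs scaled by Param4.
# Opacity fades out as the camera nears via pixel depth, plus depth fade, with both skipped on low-end feature levels.
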Material /Figment_S06/Athena/Items/QuestInteractables/SkydiveRings/FX/M_Athena_SkydiveRings_Sprite
graph {
 properties {
  bUsedWithParticleSprites=true
  blendMode=BLEND_Translucent
  pixelDepthOffsetMode=PDOM_Legacy
  shadingModel=MSM_Unlit
  twoSided=true
  usedShadingModels=Unlit
 }
 nodes [
  n1 = PixelDepth()
  n2 = Constant() {r=10000}
  n3 = Clamp(None=Divide(A=Subtract(A=@n1.R, B=@n2), B=Constant() {r=10000}))
  n4 = TextureCoordinate()
  n5 = Constant2Vector() {g=0.5, r=0.5}
  n6 = DynamicParameter() {defaultValue=<rgba 0.5 0.6 1 1>, paramNames [Radius,Radius_Edge,Param3,Param4]}
  n7 = ScalarParameter() {defaultValue=-1, parameterName=Hardness}
  n8 = SphereMask(
   A=TextureCoordinate()
   B=Constant2Vector() {g=0.5, r=0.5}
   Radius=DynamicParameter() {defaultValue=<rgba 0.5 0.6 1 1>, paramNames [Radius,Radius_Edge,Param3,Param4]}.Radius_Edge
   Hardness=ScalarParameter() {defaultValue=0.5, parameterName=Hardness_Edge}
  )
  n9 = Multiply(A=ParticleColor().A, B=Multiply(A=SphereMask(A=@n4, B=@n5, Radius=@n6.Radius, Hardness=@n7), B=OneMinus(None=@n8)))
  n10 = TextureCoordinate() {uTiling=0.5, vTiling=0.5}
  n11 = DynamicParameter() {defaultValue=<rgba 0.5 0.6 1 1>, paramNames [Radius,Radius_Edge,Param3,Param4]}
  n12 = Constant2Vector() {g=-0.25, r=-0.25}
  n13 = MaterialFunctionCall!(UVs=Add(A=Multiply(A=@n10, B=@n11.Param4), B=Multiply(A=@n12, B=@n11.Param4))) {materialFunction=/Game/Effects/Fort_Effects/Materials/_Functions/VectortoRadialCorrection}
  n14 = Rotator(
   Coordinate=Panner(Coordinate=@n13.Vector to Radial Reversed) {speedX=-0.25, speedY=1.25}
   Time=ScalarParameter() {defaultValue=-1, parameterName=RotateSpiral_2}
  )
  n15 = Rotator(
   Coordinate=Panner(Coordinate=@n13.Vector to Radial) {speedX=0.25, speedY=1.25}
   Time=ScalarParameter() {defaultValue=1, parameterName=RotateSpiral_1}
  )
  n16 = Add(
   A=TextureSample(UVs=Panner(Coordinate=@n13.Vector to Radial) {speedY=1.2}) {texture=/Game/Effects/Fort_Effects/Textures/Tile/T_CloudScape_08}.G
   B=LinearInterpolate(
    A=TextureSample(UVs=@n14) {samplerType=SAMPLERTYPE_LinearColor, texture=/Game/Effects/Fort_Effects/World/Enviroment/Lightning/lightningPainting}.G
    B=TextureSample(UVs=@n15) {samplerType=SAMPLERTYPE_LinearColor, texture=&lightningPainting}.G
    Alpha=@n13.Vector to Reversed Mask
   )
  )
  n17 = LinearInterpolate(
   A=@n9
   B=Multiply(A=@n9, B=Power(Base=@n16))
   Alpha=DynamicParameter() {defaultValue=<rgba 0.5 0.6 1 1>, paramNames [Radius,Radius_Edge,Param3,Param4]}.Param3
  )
  n18 = DepthFade(Opacity=@n17) {fadeDistanceDefault=1}
  Multiply(
   A=FeatureLevelSwitch(Default=@n3, ES2=Constant() {r=1}, SM4_REMOVED=@n3)
   B=FeatureLevelSwitch(Default=@n18, ES2=@n17, ES3_1=@n17, SM4_REMOVED=@n18)
  )
  ParticleColor()
  ScalarParameter() {defaultValue=0.5, parameterName=Radius}
  ScalarParameter() {defaultValue=0.6, parameterName=Radius_Edge}
 ]
}
