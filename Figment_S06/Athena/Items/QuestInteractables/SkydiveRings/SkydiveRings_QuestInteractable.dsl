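; Skydive ring quest interactable that grants a quest objective update when a player overlaps it outside Creative, then plays success effects and hides on the client.
; BeginPlay tints the ring material and particles by height level and scales mesh and overlap sphere, with a curve-table multiplier for hotfix sizing.
; After a delay or playlist load, clients hide rings lacking a valid objective or in Creative unless testing, and bind skydive end to hide them outside Warmup.
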(event EventBeginPlay
  (bind _heightscalemulitplier (Variables|Default|GetHeightScaleMulitplier))
  (bind _scalemultiplier (Variables|RingScale|GetScaleMultiplier))
  (bind _returnvalue (* _scalemultiplier _heightscalemulitplier))
  (bind _sm_athena_skydivering (Variables|Default|GetSM_Athena_SkydiveRing))
  (bind _relativescale3d (Variables|/Script/Engine.SceneComponent|GetRelativeScale3D _sm_athena_skydivering))
  (bind _returnvalue_2 (* _relativescale3d _returnvalue))
  (bind _highestlevel (Variables|Default|GetHighestLevel))
  (bind _ringmat (Variables|Default|GetRingMat))
  (bind _middlelevel (Variables|Default|GetMiddleLevel))
  (bind _lowestlevel (Variables|Default|GetLowestLevel))
  (bind _p_athena_skydiverings (Variables|Default|GetP_Athena_SkydiveRings))
  ;[1 1 1] Set up Material
  (bind _returnvalue_3 (Rendering|Material|CreateDynamicMaterialInstance (Variables|Default|GetSM_Athena_SkydiveRing) 0 "/Game/Athena/Items/QuestInteractables/SkydiveRings/FX/M_Athena_SkydiveRings_Simple.M_Athena_SkydiveRings_Simple" "None" :!class /Script/Engine.PrimitiveComponent))
  (Variables|Default|SetRingMat _returnvalue_3)
  (Utilities|FlowControl|Sequence
    (:then_0
      ;[1 1 1] Hotfix Size Adjustments
      (bind _outxy (Fortnite|EvaluateCurveTableRow (Variables|RingScale|GetSizeScalerRow)))
      (Variables|RingScale|SetScaleMultiplier _outxy)
      ;[1 1 1] Check Height Level
      (switch int (Variables|Default|GetHeightLevel)
        (:0
          ;[1 1 1] Set Color according to Height Level
          (Rendering|Material|SetVectorParameterValue _ringmat "ColorInner" _lowestlevel :!class /Script/Engine.MaterialInstanceDynamic)
          (Effects|Components|ParticleSystem|SetColorParameter _p_athena_skydiverings "PulseRingColor" _lowestlevel)
          (Variables|Default|SetHeightScaleMulitplier 1.0))
        (:1
          (Rendering|Material|SetVectorParameterValue _ringmat "ColorInner" _middlelevel :!class /Script/Engine.MaterialInstanceDynamic)
          (Effects|Components|ParticleSystem|SetColorParameter _p_athena_skydiverings "PulseRingColor" _middlelevel)
          (Variables|Default|SetHeightScaleMulitplier 1.25))
        (:2
          (Rendering|Material|SetVectorParameterValue _ringmat "ColorInner" _highestlevel :!class /Script/Engine.MaterialInstanceDynamic)
          (Effects|Components|ParticleSystem|SetColorParameter _p_athena_skydiverings "PulseRingColor" _highestlevel)
          (Variables|Default|SetHeightScaleMulitplier 1.5)))
      ;[1 1 1] Set Scale according to HeightLevel
      (Variables|RingScale|SetAdjustedRingScale _returnvalue_2)
      (Transformation|SetRelativeScale3D _sm_athena_skydivering _returnvalue_2)
      (Components|Sphere|SetSphereRadius (Variables|Default|GetOverlapSphere) (* _returnvalue (Components|Sphere|GetScaledSphereRadius (Variables|Default|GetOverlapSphere))) true))
    (:then_1
      ;[1 0.6341 0.1593] Need Something Better from Code
      (Utilities|FlowControl|Delay 5.0)
      ;[1 1 1] CLIENT AND SERVER - Check if In Creative Mode or not
      (bind _asfort_game_state_athena (Utilities|Casting|CastToFortGameStateAthena (Game|GetGameState)))
      (if (Variables|/Script/FortniteGame.FortGameStateAthena|GetbPlaylistDataIsLoaded _asfort_game_state_athena)
        (if (GameplayTags|HasTag (Class|FortGameStateAthena|GetAthenaPlaylistContextTags _asfort_game_state_athena) "(TagName=\"Athena.Creative\")")
          (Variables|Default|SetCreativeMode true)
          (else
            (Variables|Default|SetCreativeMode false)))
        ;[1 1 1] Does Player have quest?
        (Networking|SwitchHasAuthority
          (:Remote
            (bind _isvalid (CallFunction|IsValidObjective))
            (if (and _isvalid (not (Variables|Default|GetCreativeMode)))
              _
              (elif (Variables|Default|GettestingEnabled)
                _
                (else
                  ;[1 1 1] Hide Ring when landing after skydive or if quest is invalid
                  (Development|PrintString "OBJ not vaild" (Variables|Default|GetDebugOn) (Variables|Default|GetDebugOn))
                  (CallFunction|HideRing))))
            ;[1 1 1] Bind Skydive and RingTriggered Events
            (bind _asathena_player_controller (Utilities|Casting|CastToAthena_PlayerController (Utilities|Array|Get(acopy) (Fortnite|GetLocalFortPlayerControllers true))))
            (CallFunction|BeginSkyDiving)
            (EventDispatchers|BindEventtoOnEndSkydiving _asathena_player_controller (EventDispatchers|CreateEvent :!event Custom|LandedFromSkydive))))
        (else
          (Playlist|BindEventtoOnCurrentPlaylistLoaded _asfort_game_state_athena (delegate Custom|PlaylistLoaded)))))))

(event Custom|PlaylistLoaded
  (params (PlaylistName :type Name) (PlaylistContextTags :type Gameplay%20Tag%20Container%20Structure%20%28by%20ref%29))
  ;[1 1 1] CLIENT AND SERVER - Check if In Creative Mode or not
  (if (GameplayTags|HasTag PlaylistContextTags "(TagName=\"Athena.Creative\")")
    (Variables|Default|SetCreativeMode true)
    (else
      (Variables|Default|SetCreativeMode false)))
  ;[1 1 1] Does Player have quest?
  (Networking|SwitchHasAuthority
    (:Remote
      (bind _isvalid (CallFunction|IsValidObjective))
      (if (and _isvalid (not (Variables|Default|GetCreativeMode)))
        _
        (elif (Variables|Default|GettestingEnabled)
          _
          (else
            ;[1 1 1] Hide Ring when landing after skydive or if quest is invalid
            (Development|PrintString "OBJ not vaild" (Variables|Default|GetDebugOn) (Variables|Default|GetDebugOn))
            (CallFunction|HideRing))))
      ;[1 1 1] Bind Skydive and RingTriggered Events
      (bind _asathena_player_controller (Utilities|Casting|CastToAthena_PlayerController (Utilities|Array|Get(acopy) (Fortnite|GetLocalFortPlayerControllers true))))
      (CallFunction|BeginSkyDiving)
      (EventDispatchers|BindEventtoOnEndSkydiving _asathena_player_controller (EventDispatchers|CreateEvent :!event Custom|LandedFromSkydive)))))

(event Custom|LandedFromSkydive
  ;[1 1 1] Hide Ring when landing after skydive or if quest is invalid
  (Utilities|Casting|CastToFortGameStateAthena (Game|GetGameState))
  (if (not (Utilities|Enum|Equal(Enum) (BattleRoyaleGamePhaseLogic|GetGamePhase (BattleRoyaleGamePhaseLogic|GetBattleRoyaleGamePhaseLogicComponent)) (Utilities|Enum|LiteralenumEAthenaGamePhase "Warmup")))
    (CallFunction|HideRing)))

(event Custom|BeginSkyDiving)

(event Collision|EventActorBeginOverlap
  (params (OtherActor :type Actor%20Object%20Reference))
  ;[1 1 1] On overlap
  (bind _asplayer_pawn_athena (Utilities|Casting|CastToPlayerPawn_Athena OtherActor))
  (if (not (Variables|Default|GetCreativeMode))
    (bind _asathena_player_controller (Utilities|Casting|CastToAthena_PlayerController (Game|GetInstigatorController OtherActor))
      (:then
        (Quest|GiveQuestUpdatetoPlayers (Variables|QuestStuff|GetQuestItem) (Variables|QuestStuff|GetQuestObjectiveBackendName) (Variables|QuestStuff|GetRingsObjectiveStat) (Utilities|Array|MakeArray _asathena_player_controller))
        ;[1 1 1] Client Only
        (Networking|SwitchHasAuthority
          (:Remote
            (Development|PrintString "Ring Triggered" (Variables|Default|GetDebugOn) (Variables|Default|GetDebugOn))
            ;[1 1 1] FX for Success
            (Audio|PlaySoundatLocation (Variables|Default|GetSuccessSound))
            (Effects|SpawnEmitterAttached (Variables|Default|GetSuccessEmitter) (Variables|/Script/Engine.Character|GetMesh _asplayer_pawn_athena))
            (Components|Activation|Activate (Variables|Default|GetForceFeedback) false)
            (CallFunction|HideRing))))
      (:CastFailed
        (Development|PrintString "CAST FAILED: SKYDIVE RINGS")))))

(event Custom|HideRing
  ;[1 1 1] Hide Ring
  (Components|Activation|Deactivate (Variables|Default|GetP_Athena_SkydiveRings))
  (Rendering|SetActorHiddenInGame true)
  (Collision|SetActorEnableCollision false))

(fn IsValidObjective
  (bind _questvalid (Quest|DoesPlayerHaveQuest (Utilities|Array|Get(acopy) (Fortnite|GetLocalFortPlayerControllers true)) (Variables|QuestStuff|GetQuestItem) (Variables|QuestStuff|GetQuestObjectiveBackendName)))
  (return _questvalid))

(fn OrientToLocalPlayer
  (bind _returnvalue (Fortnite|GetLocalFortPlayerControllers true))
  (bind _returnvalue_2 (Transformation|GetActorLocation))
  (bind _output (Utilities|Array|Get(acopy) _returnvalue))
  (bind _returnvalue_3 (Transformation|GetActorLocation _output))
  (Networking|SwitchHasAuthority
    (:Remote
      (if (> (.z _returnvalue_3) (.z _returnvalue_2))
        (Transformation|SetWorldRotation (Variables|Default|GetSM_Athena_SkydiveRing) (Math|Rotator|MakeRotfromZ (/ (+ (Variables|Default|GetStraightUp) (Math|Vector|Normalize (- _returnvalue_3 _returnvalue_2))) 2.0)) false false)))))

(fn ConstructionScript)
